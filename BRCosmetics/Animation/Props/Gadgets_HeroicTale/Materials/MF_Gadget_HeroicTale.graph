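# Material function for the HeroicTale gadget emote prop that layers a burning, cracking dissolve onto incoming material attributes.
# Dissolve and emissive progress parameters, world-aligned textures sampled in pre-skinned local space and a UV-tiled crack mask build large and small cracks.
# These shape opacity mask, glowing edges, burnt base color and blended normals.
# Sphere masks localize the effect, while world position offset adds directional push, normal inflation and time-driven sine wobble.
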MaterialFunction /BRCosmetics/Animation/Props/Gadgets_HeroicTale/Materials/MF_Gadget_HeroicTale
graph {nodes [
  Frac(None=Multiply(A=Time()) {constB=0.5})
  n1 = GetMaterialAttributes(None=FunctionInput() {inputName=Material, inputType=FunctionInput_MaterialAttributes}) {attributeGetTypes [#1, #2, #3]}
  n2 = SetMaterialAttributes!(
   MaterialAttributes=@n1.MaterialAttributes
   EmissiveColor=NamedRerouteUsage()
   OpacityMask=NamedRerouteUsage()
   WorldPositionOffset=NamedRerouteUsage()
   BaseColor=LinearInterpolate(
    A=@n1.BaseColor
    B=VectorParameter() {group=HeroicTale, parameterName=HT_BurntColor, defaultValue=<rgba 0.0365 0.0058 0.001 1>}.RGB
    Alpha=NamedRerouteUsage()
   )
   Normal
  ) {attributeSetTypes [
    #1
    #4
    #5
    #2
    #3
   ]}
  FunctionOutput(None=@n2)
  NamedRerouteDeclaration(None=ScalarParameter() {defaultValue=1, group=HeroicTale, parameterName=HT_DissolveProgress}) {name=DissolveProgress, nodeColor=<rgba 1 0.4388 0.2554 1>}
  n3 = ScalarParameter() {defaultValue=200, group=HeroicTale, parameterName=HT_TexSize}
  n4 = VertexInterpolator(VS=LocalPosition() {localOrigin=InstancePreSkinning}.XYZ)
  n5 = VertexInterpolator(VS=PreSkinnedNormal())
  n6 = MaterialFunctionCall(
   TextureObject=TextureObject() {samplerType=SAMPLERTYPE_Masks, texture=/BRCosmetics/Animation/Props/Gadgets_HeroicTale/Textures/T_HeroicTale_Emote_Prop_TA}
   TextureSize=@n3
   WorldPosition=@n4.PS
   World Space Normal=@n5.PS
   ProjectionTransitionContrast=Constant() {r=1}
  ) {materialFunction=&WorldAlignedTexture}
  NamedRerouteDeclaration(None=ComponentMask(None=@n6.XYZ Texture) {b=true}) {name=OpacityGradient, nodeColor=<rgba 1 0 0.75 1>}
  NamedRerouteDeclaration(None=@n1.Normal) {name=Normal_IN, nodeColor=<rgba 1 0 0.9844 1>}
  n7 = Constant3Vector() {constant=<rgba 0 20 30 0>}
  n8 = WorldPosition() {worldPositionShaderOffset=WPT_CameraRelative}
  n9 = CameraPositionWS()
  n10 = NamedRerouteUsage()
  n11 = Saturate(None=SphereMask(
    A=TransformPosition(None=@n7) {transformType=TRANSFORMPOSSOURCE_World}
    B=Add(A=@n8.XYZ, B=@n9)
    Radius=LinearInterpolate(Alpha=@n10) {constB=300}
   ) {hardnessPercent=0})
  n12 = OneMinus(None=ComponentMask(None=@n6.XYZ Texture) {r=true})
  n13 = NamedRerouteUsage()
  n14 = ScalarParameter() {defaultValue=4, group=HeroicTale, parameterName=HT_DissolvePower}
  n15 = Multiply(
   A=Add(A=SmoothStep(Min=@n11, Value=@n12) {constMax=0, constMin=0.5}, B=Power(Base=@n13, Exp=@n14) {constExponent=0.5})
   B=VectorParameter() {group=HeroicTale, parameterName=HT_CracksColor, defaultValue=<rgba 20 5.9825 0 1>}.RGB
  )
  n16 = SmoothStep(Min=Constant() {r=2}, Max=Constant(), Value=NamedRerouteUsage()) {constMin=0.5}
  NamedRerouteDeclaration(None=Add(A=NamedRerouteUsage(), B=LinearInterpolate(A=@n15, Alpha=@n16) {constB=0})) {name=Emissive_OUT, nodeColor=<rgba 1 0 0.9375 1>}
  n17 = SphereMask(
   A=TransformPosition(None=Constant3Vector() {constant=<rgba 0 20 30 0>}) {transformType=TRANSFORMPOSSOURCE_World}
   B=Add(A=WorldPosition() {worldPositionShaderOffset=WPT_CameraRelative}.XYZ, B=CameraPositionWS())
   Radius=LinearInterpolate(B=Constant() {r=150}, Alpha=NamedRerouteUsage()) {constB=100}
  ) {hardnessPercent=5}
  n18 = VertexNormalWS()
  n19 = VectorParameter() {group=HeroicTale, parameterName=HT_InflateAmount, defaultValue=<rgba 2 2 0.1 1>}
  n20 = Multiply(
   A=SmoothStep(Value=NamedRerouteUsage()) {constMin=0.5}
   B=Add(A=VectorParameter() {group=HeroicTale, parameterName=HT_WPO_Amount, defaultValue=<rgba 0 0 -40 1>}.RGB, B=Multiply(A=@n18, B=@n19.RGB))
  )
  n21 = WorldPosition()
  n22 = SmoothStep(Value=NamedRerouteUsage()) {constMin=0.5}
  n23 = Frac(None=Multiply(A=Time(), B=Multiply(A=@n22) {constB=2}))
  n24 = Sine(None=Multiply(A=Add(A=@n21.XYZ, B=@n23), B=ScalarParameter() {defaultValue=3, group=HeroicTale, parameterName=HT_WPOWaveSize}))
  n25 = NamedRerouteUsage()
  NamedRerouteDeclaration(None=Add(A=Multiply(A=@n17, B=@n20), B=Multiply(A=@n24, B=@n25))) {name=WPO, nodeColor=<rgba 0.5703 0 1 1>}
  NamedRerouteDeclaration(None=ScalarParameter() {defaultValue=0.7, group=HeroicTale, parameterName=HT_EmissiveProgress, sliderMax=1}) {name=EmissiveProgress, nodeColor=<rgba 0 1 0.625 1>}
  n26 = Add(A=ConstantBiasScale(None=NamedRerouteUsage()) {bias=-0.5, scale=2}, B=NamedRerouteUsage())
  n27 = Power(Base=@n26, Exp=Constant() {r=0.7})
  n28 = TextureSample(UVs=Multiply(A=TextureCoordinate(), B=ScalarParameter() {defaultValue=2, group=HeroicTale, parameterName=HT_Tiling})) {
   samplerSource=SSM_Wrap_WorldGroupSettings
   samplerType=SAMPLERTYPE_Masks
   texture=&T_HeroicTale_Emote_Prop_TA
  }
  n29 = Add(
   A=NamedRerouteDeclaration(None=SmoothStep(Value=@n26) {constMin=0.5}) {name=LargeCracks, nodeColor=<rgba 0.0078 1 0 1>}
   B=NamedRerouteDeclaration(None=Multiply(A=@n27, B=@n28.G)) {name=SmallCracks, nodeColor=<rgba 1 0 0.3281 1>}
  )
  NamedRerouteDeclaration(None=Saturate(None=OneMinus(None=@n29))) {name=OpacityMask, nodeColor=<rgba 1 0.3047 0 1>}
  n30 = Add(A=ConstantBiasScale(None=NamedRerouteUsage()) {bias=-0.5, scale=2}, B=NamedRerouteUsage())
  n31 = ScalarParameter() {defaultValue=0.5, group=HeroicTale, parameterName=HT_SmallCracksEdgeGlowWidth}
  n32 = NamedRerouteUsage()
  n33 = NamedRerouteUsage()
  n34 = NamedRerouteUsage()
  n35 = Multiply(A=LinearInterpolate(A=@n17, B=Add(A=@n33, B=@n34), Alpha=@n34), B=@n17)
  n36 = SmoothStep(
   Min=Add(A=@n28.G) {constB=-0.5}
   Max=Add(A=@n28.G, B=LinearInterpolate(B=@n31, Alpha=@n32) {constB=2}) {constB=0.5}
   Value=OneMinus(None=Multiply(A=@n35, B=@n30))
  )
  n37 = Multiply(A=OneMinus(None=Saturate(None=@n36)), B=Power(Base=@n30))
  NamedRerouteDeclaration(None=Saturate(None=Add(A=@n30, B=@n37) {constB=0})) {name=DissolveGlowMask, nodeColor=<rgba 0 0.7656 1 1>}
  NamedRerouteDeclaration(None=@n1.EmissiveColor) {name=Emissive_IN, nodeColor=<rgba 0.6172 1 0 1>}
  n38 = Constant3Vector() {constant=<rgba 0 0 1 0>}
  n39 = MaterialFunctionCall(
   Normal=MaterialFunctionCall!(
    TextureObject=TextureObject() {samplerType=SAMPLERTYPE_Normal, texture=/BRCosmetics/Animation/Props/Gadgets_HeroicTale/Textures/T_HeroicTale_Emote_Prop_TA_N}
    TextureSize=@n3
    WorldSpace
    WorldNormal=@n5.PS
    WorldPosition=@n4.PS
    Use High Quality Normals=StaticBool() {value=true}
   ) {materialFunction=/Volumetrics/Content/PlanetPrototype/Functions/WorldAlignedNormal_HQ}.XYZ Texture
   Flatness=Constant() {r=-5}
  ) {materialFunction=&FlattenNormal}
  n40 = NamedRerouteUsage()
  NamedRerouteDeclaration(None=MaterialFunctionCall(BaseNormal=LinearInterpolate(A=@n38, B=@n39.Result, Alpha=@n40), AdditionalNormal=NamedRerouteUsage()) {materialFunction=&BlendAngleCorrectedNormals}.Result) {name=Normal_OUT, nodeColor=<rgba 0.6641 1 0 1>}
  NamedRerouteUsage()
 ]}
